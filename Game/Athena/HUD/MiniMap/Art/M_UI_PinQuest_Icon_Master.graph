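# Translucent UI master material for a pin-quest icon in the minimap HUD, layering icon, outline and optional backing plate.
# One texture's red channel is smoothstepped at separate thresholds for icon and outline, with a radial gradient filling the outline's center.
# A Selected scalar blends icon and outline from base to hover colors.
# Opacity combines these masks with a rotated, resolution-corrected angled plate edge, toggled by IsPlateOn.
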Material /Game/Athena/HUD/MiniMap/Art/M_UI_PinQuest_Icon_Master
graph {
 properties {blendMode=BLEND_Translucent, materialDomain=MD_UI, numCustomizedUVs=1, pixelDepthOffsetMode=PDOM_Legacy, usedShadingModels=Default Lit}
 nodes [
  n1 = VectorParameter() {group=Outline, parameterName=OutlineColorBase, defaultValue=<rgba 1 0 0 0>}
  n2 = VectorParameter() {group=Outline, parameterName=OutlineColorHover, defaultValue=<rgba 0.2818 1 0 1>}
  n3 = ScalarParameter() {parameterName=Selected}
  n4 = VectorParameter() {parameterName="CenterPosition (V2)", defaultValue=<rgba 0.5 0.45 0.4 0>}
  n5 = ComponentMask(None=@n4.RGB) {g=true, r=true}
  n6 = ScalarParameter() {defaultValue=0.2, parameterName=CenterFill}
  n7 = ScalarParameter() {defaultValue=0.005, group=Outline, parameterName=Outline}
  n8 = ScalarParameter() {defaultValue=0.035, group=Outline, parameterName=EdgeSoftness}
  n9 = TextureSample() {samplerSource=SSM_Clamp_WorldGroupSettings, texture=&/T_UI_PinQuest_Icon}
  n10 = Max(
   A=MaterialFunctionCall(
    Alpha=MaterialFunctionCall(CenterPosition=@n5, Radius=@n4.B) {materialFunction=&RadialGradientExponential}.RadialGradientExponential
    Min=@n6
    Max=@n6
   ) {materialFunction=&SmoothStep}.Result
   B=MaterialFunctionCall(Alpha=@n9.R, Min=@n7, Max=Add(A=@n7, B=@n8)) {materialFunction=&SmoothStep}.Result
  )
  n11 = ScalarParameter() {defaultValue=0.25, group=Icon, parameterName=Edge}
  n12 = MaterialFunctionCall(Alpha=@n9.R, Min=@n11, Max=Add(A=@n11, B=ScalarParameter() {defaultValue=0.035, group=Icon, parameterName=IconEdge})) {materialFunction=&SmoothStep}
  LinearInterpolate(
   A=LinearInterpolate(
    A=VectorParameter() {group=BackingPlate, parameterName=PlateColor, defaultValue=<rgba 0.0818 0 1 1>}.RGB
    B=LinearInterpolate(A=@n1.RGB, B=@n2.RGB, Alpha=@n3)
    Alpha=@n10
   )
   B=LinearInterpolate(
    A=VectorParameter() {group=Icon, parameterName=IconColorBase, defaultValue=<rgba 0 0.9546 1 0>}.RGB
    B=VectorParameter() {group=Icon, parameterName=IconColorHover, defaultValue=<rgba 0 0.0508 1 0>}.RGB
    Alpha=ScalarParameter() {parameterName=Selected}
   )
   Alpha=@n12.Result
  )
  n13 = Max(B=Max(A=@n12.Result, B=@n10))
  n14 = MaterialFunctionCall!(Rotation_Fraction=ScalarParameter() {defaultValue=0.126, group=BackingPlate, parameterName=Plate_Angle}, TexCoord=TextureCoordinate()) {materialFunction=/FortUILibrary/Foundation/Art_ToolKit/AngledBoxes/MF_UI_Rotator_Fractions}
  n15 = AppendVector(A=Constant(), B=ScalarParameter() {defaultValue=-0.495, group=BackingPlate, parameterName=Offset_Angle_Top})
  n16 = Saturate(None=ComponentMask(None=Add(A=@n14.Result, B=@n15)) {g=true})
  n17 = Multiply(
   A=MaterialFunctionCall!(Native_U_PixelSize=ScalarParameter() {defaultValue=120, group=BackingPlate, parameterName=Widget_X_Dimension}) {materialFunction=/FortUILibrary/Foundation/Art_ToolKit/Tools/M_UI_Resolution_Sharpness_Correction}.Result
   B=ScalarParameter() {defaultValue=16, group=BackingPlate, parameterName=Angle_Edge_Softness}
  )
  n18 = LinearInterpolate(B=Saturate(None=Multiply(A=@n16, B=@n17)), Alpha=ScalarParameter() {group=BackingPlate, parameterName=IsPlateOn})
  Saturate(None=Multiply(A=Add(A=@n13, B=@n18)))
 ]
}
