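# M_UI_QuestPin is a translucent UI material that draws a minimap quest pin with tintable backing, outline and icon.
# Channel-packed pin and number textures supply masks, and smoothstep-driven lerp chains select unavailable, blank, 1–3 or star icons without branching.
# LocationType drives an alternate procedural shape built from a rotatable rounded mask.
# Static switches toggle outline, backing and an override icon, which has adjustable offset, scale and channel selection.
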Material /Game/Athena/HUD/MiniMap/Art/M_UI_QuestPin
graph {
 properties {bAutomaticallySetUsageInEditor=false, blendMode=BLEND_Translucent, materialDomain=MD_UI, usedShadingModels=Default Lit}
 nodes [
  n1 = NamedRerouteUsage()
  n2 = TextureSampleParameter2D(UVs=NamedRerouteUsage()) {parameterName=QuestPin, samplerType=SAMPLERTYPE_LinearColor, texture=&/T_UI_QuestPin}
  n3 = Saturate(None=NamedRerouteUsage())
  n4 = SmoothStep(Value=LinearInterpolate(A=@n2.RGBA, B=Add(A=@n2.RGBA, B=@n1), Alpha=@n3))
  n5 = NamedRerouteUsage()
  NamedRerouteDeclaration(None=LinearInterpolate(A=ComponentMask(None=@n4) {r=true}, B=NamedRerouteUsage(), Alpha=Saturate(None=@n5))) {name=Outline, nodeColor=<rgba 0 0.3235 1 1>}
  NamedRerouteDeclaration(None=ComponentMask(None=@n4) {g=true}) {name=Star, nodeColor=<rgba 0 0.3235 1 1>}
  n6 = VectorParameter() {parameterName=OutlineColor, defaultValue=<rgba 1 1 1 1>}
  NamedRerouteDeclaration(None=@n6.RGB) {name=OutlineColor, nodeColor=<rgba 1 0.8879 0 1>}
  n7 = VectorParameter() {parameterName=IconColor, defaultValue=<rgba 1 1 1 1>}
  NamedRerouteDeclaration(None=@n7.RGB) {name=IconColor, nodeColor=<rgba 1 0.8879 0 1>}
  n8 = VectorParameter() {parameterName=BackingColor, defaultValue=<rgba 0 0 0 1>}
  NamedRerouteDeclaration(None=@n8.A) {name=BackingOpacity, nodeColor=<rgba 1 0.8879 0 1>}
  n9 = LinearInterpolate(A=NamedRerouteUsage(), B=NamedRerouteUsage(), Alpha=SmoothStep(Value=NamedRerouteUsage()) {constMax=0, constMin=-1}) {constA=-1, constB=-0.99}
  n10 = NamedRerouteUsage()
  n11 = SmoothStep(Value=NamedRerouteUsage()) {constMin=0.1}
  n12 = NamedRerouteUsage()
  n13 = LinearInterpolate(
   A=LinearInterpolate(A=LinearInterpolate(A=@n9, B=@n10, Alpha=@n11), B=NamedRerouteUsage(), Alpha=SmoothStep(Value=@n12) {constMax=2, constMin=1.1})
   B=NamedRerouteUsage()
   Alpha=SmoothStep(Value=NamedRerouteUsage()) {constMax=3, constMin=2.1}
  )
  n14 = NamedRerouteUsage()
  n15 = SmoothStep(Value=NamedRerouteUsage()) {constMax=4, constMin=3.1}
  n16 = NamedRerouteUsage()
  n17 = OneMinus(None=NamedRerouteUsage())
  n18 = NamedRerouteUsage()
  NamedRerouteDeclaration(None=LinearInterpolate(A=LinearInterpolate(A=@n13, B=@n14, Alpha=@n15), B=Multiply(A=@n16, B=@n17), Alpha=SmoothStep(Value=@n18) {constMax=5, constMin=4.1})) {name=Icon, nodeColor=<rgba 0 0.3235 1 1>}
  NamedRerouteDeclaration(None=StaticBoolParameter() {defaultValue=true, parameterName=HasOutline}) {name=HasOutline, nodeColor=<rgba 0.2812 0 0.0174 1>}
  NamedRerouteDeclaration(None=StaticBoolParameter() {defaultValue=true, parameterName=HasBacking}) {name=HasBacking, nodeColor=<rgba 0.2812 0 0.0174 1>}
  n19 = NamedRerouteUsage()
  n20 = NamedRerouteUsage()
  n21 = StaticSwitch(
   True=LinearInterpolate(A=NamedRerouteUsage(), B=NamedRerouteUsage(), Alpha=Max(A=@n19, B=@n20))
   False=NamedRerouteUsage()
   Value=NamedRerouteUsage()
  )
  n22 = NamedRerouteUsage()
  n23 = Max(A=NamedRerouteUsage(), B=NamedRerouteUsage())
  NamedRerouteDeclaration(None=StaticSwitch(True=LinearInterpolate(A=@n21, B=@n22, Alpha=@n23), False=@n21, Value=NamedRerouteUsage())) {name=FinalColor, nodeColor=<rgba 0 0.3235 1 1>}
  n24 = SmoothStep(Value=NamedRerouteUsage()) {constMin=0.97}
  NamedRerouteDeclaration(None=OneMinus(None=@n24)) {name=PinAlpha, nodeColor=<rgba 0 0.3235 1 1>}
  NamedRerouteDeclaration(None=@n7.A) {name=IconOpacity, nodeColor=<rgba 1 0.8879 0 1>}
  NamedRerouteDeclaration(None=@n6.A) {name=OutlineOpacity, nodeColor=<rgba 1 0.8879 0 1>}
  NamedRerouteDeclaration(None=ComponentMask(None=@n4) {b=true}) {name=Unavailable, nodeColor=<rgba 0 0.3235 1 1>}
  n25 = SmoothStep(Value=NamedRerouteUsage()) {constMax=0.5}
  NamedRerouteDeclaration(None=Min(A=@n24, B=OneMinus(None=@n25))) {name=OutlinePinAlpha, nodeColor=<rgba 0 0.3235 1 1>}
  n26 = NamedRerouteUsage()
  n27 = NamedRerouteUsage()
  n28 = NamedRerouteUsage()
  NamedRerouteDeclaration(None=Multiply(A=StaticSwitch(True=@n26, False=@n27, Value=@n28), B=NamedRerouteUsage())) {name=BackingAlpha, nodeColor=<rgba 0 0.3235 1 1>}
  NamedRerouteDeclaration(None=Multiply(A=NamedRerouteUsage(), B=NamedRerouteUsage())) {name=IconAlpha, nodeColor=<rgba 0 0.3235 1 1>}
  NamedRerouteDeclaration(None=StaticBoolParameter() {parameterName=HasOverrideIcon}) {name=HasIconOverride, nodeColor=<rgba 0.2812 0 0.0174 1>}
  NamedRerouteDeclaration(None=MaterialFunctionCall!(UVs=NamedRerouteUsage(), Alpha=NamedRerouteUsage()) {materialFunction=/Game/UI/Foundation/Items/ClampAlphaToTransparentBorder}.Result) {name=OverrideIcon, nodeColor=<rgba 0 0.3235 1 1>}
  n29 = NamedRerouteUsage()
  n30 = AppendVector(A=ScalarParameter() {parameterName=OverrideIcon_XOffset}, B=ScalarParameter() {defaultValue=0.05, parameterName=OverrideIcon_YOffset})
  NamedRerouteDeclaration(None=MaterialFunctionCall(UVs=Add(A=@n29, B=@n30), Texture Scale=ScalarParameter() {defaultValue=0.65, parameterName=OverrideIconScale}) {materialFunction=&ScaleUVsByCenter}.UVs) {name=OverrideIconUVs, nodeColor=<rgba 0 0.3235 1 1>}
  n31 = TextureSampleParameter2D(UVs=NamedRerouteUsage()) {parameterName=Icon, texture=/Game/UI/Foundation/Textures/Icons/Replay/T-Icon-Replay-Exclamation-L}
  NamedRerouteDeclaration(None=@n31.R) {name=OverrideIcon_RedChannel, nodeColor=<rgba 0 0.3513 1 1>}
  NamedRerouteDeclaration(None=@n31.G) {name=OverrideIcon_GreenChannel, nodeColor=<rgba 0 0.3513 1 1>}
  NamedRerouteDeclaration(None=@n31.B) {name=OverrideIcon_BlueChannel, nodeColor=<rgba 0 0.3513 1 1>}
  NamedRerouteDeclaration(None=@n31.A) {name=OverrideIcon_AlphaChannel, nodeColor=<rgba 0 0.3513 1 1>}
  NamedRerouteDeclaration(None=LinearInterpolate(A=@n2.A, B=NamedRerouteUsage(), Alpha=@n3)) {name=Plain, nodeColor=<rgba 0 0.3235 1 1>}
  n32 = ScalarParameter() {parameterName=IconChannelSelector}
  n33 = LinearInterpolate(A=NamedRerouteUsage(), B=NamedRerouteUsage(), Alpha=SmoothStep(Value=@n32)) {constA=-1, constB=-0.99}
  n34 = NamedRerouteUsage()
  n35 = SmoothStep(Value=@n32) {constMax=3, constMin=2.1}
  NamedRerouteDeclaration(None=LinearInterpolate(A=LinearInterpolate(A=@n33, B=@n34, Alpha=@n35), B=NamedRerouteUsage(), Alpha=SmoothStep(Value=@n32) {constMax=2, constMin=1.1})) {name=OverrideIcon_Alpha, nodeColor=<rgba 0 0.3513 1 1>}
  n36 = TextureCoordinate()
  n37 = Constant2Vector() {g=-0.06}
  n38 = NamedRerouteUsage()
  NamedRerouteDeclaration(None=LinearInterpolate(A=@n36, B=Add(A=@n36, B=@n37), Alpha=Saturate(None=@n38))) {name=CustomShapeCoord, nodeColor=<rgba 0 0.25 1 1>}
  n39 = SmoothStep(Value=NamedRerouteUsage()) {constMax=2, constMin=1.1}
  n40 = MaterialFunctionCall(
   UV=MaterialFunctionCall(UVs=TextureCoordinate(), "Rotation Angle (0-1)"=LinearInterpolate(Alpha=@n39) {constB=0.125}) {materialFunction=&CustomRotator}.Rotated Values
   Size=LinearInterpolate(A=Constant() {r=0.8116}, Alpha=@n39) {constB=0.7}
   Corner Radius=LinearInterpolate(A=Constant() {r=10}, B=Constant() {r=0.2}, Alpha=@n39)
   Outline Thickness=ScalarParameter() {defaultValue=3.8901, parameterName=CustomOutlineThickness}
  ) {materialFunction=&MF_UI_CoreUI_RoundedMask}
  NamedRerouteDeclaration(None=@n40.Shape Mask) {name=CustomShapeMask, nodeColor=<rgba 1 0 0.3281 1>}
  NamedRerouteDeclaration(None=@n40.Outline Mask) {name=CustomShapeMaskOutline, nodeColor=<rgba 0 0.6484 1 1>}
  NamedRerouteDeclaration(None=ScalarParameter() {parameterName=LocationType}) {name=CustomShape, nodeColor=<rgba 1 0.8906 0 1>}
  NamedRerouteDeclaration(None=Subtract(A=NamedRerouteUsage(), B=@n2.A)) {name=CustomShapeDifference, nodeColor=<rgba 0 1 0.4375 1>}
  NamedRerouteDeclaration(None=ScalarParameter() {defaultValue=4, desc="-1= Unavailable, 0 = Blank, 1=1, 2=2, 3=3 ,4=star", parameterName=BackingTypeSelector}) {name=PinType, nodeColor=<rgba 0.3707 1 0.1536 1>}
  n41 = StaticSwitch(
   True=LinearInterpolate(A=NamedRerouteUsage(), B=NamedRerouteUsage(), Alpha=NamedRerouteUsage())
   False=NamedRerouteUsage()
   Value=NamedRerouteUsage()
  )
  n42 = NamedRerouteUsage()
  NamedRerouteDeclaration(None=StaticSwitch(True=Max(A=@n41, B=@n42), False=@n41, Value=NamedRerouteUsage())) {name=FinalOpacity, nodeColor=<rgba 0 0.3235 1 1>}
  n43 = TextureSampleParameter2D(UVs=NamedRerouteUsage()) {parameterName=QuestNumbers, texture=&/T_UI_QuestPin_Numbers}
  n44 = NamedRerouteUsage()
  n45 = NamedRerouteUsage()
  n46 = SmoothStep(Value=LinearInterpolate(A=@n43.RGBA, B=Add(A=@n44, B=@n43.RGBA), Alpha=Saturate(None=@n45)))
  NamedRerouteDeclaration(None=ComponentMask(None=@n46) {r=true}) {name=One, nodeColor=<rgba 0 0.3235 1 1>}
  NamedRerouteDeclaration(None=ComponentMask(None=@n46) {g=true}) {name=Two, nodeColor=<rgba 0 0.3235 1 1>}
  NamedRerouteDeclaration(None=ComponentMask(None=@n46) {b=true}) {name=Three, nodeColor=<rgba 0 0.3235 1 1>}
  NamedRerouteDeclaration(None=@n8.RGB) {name=BackingColor, nodeColor=<rgba 1 0.8879 0 1>}
  NamedRerouteUsage()
  NamedRerouteUsage()
 ]
}
