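# Translucent UI material for a quest container background in the HUD, drawing a two-color radial gradient inside a rounded shape.
# An exponential radial gradient with adjustable center, radius and density blends RGB1 and RGB2.
# The shape multiplies a CoreUI rounded mask, with per-corner radii optionally in pixels, by a sharpened MF_RoundedSquare edge.
# A second rounded-square branch on 9-slice UVs repeats the edge math separately.
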Material /Game/Athena/HUD/MiniMap/Art/M_UI_Quest_Container_MultiProduct
graph {
 properties {blendMode=BLEND_Translucent, materialDomain=MD_UI, pixelDepthOffsetMode=PDOM_Legacy, usedShadingModels=Default Lit}
 nodes [
  n1 = NamedRerouteUsage()
  n2 = AppendVector(
   A=ScalarParameter() {group=Radial gradient, parameterName=RadialGradient_u, sliderMin=0.5}
   B=ScalarParameter() {defaultValue=0.5, group=Radial gradient, parameterName=RadialGradient_V, sliderMin=0.5}
  )
  n3 = ScalarParameter() {defaultValue=1.5, group=Radial gradient, parameterName=Radial Gradient Radius_V}
  n4 = ScalarParameter() {defaultValue=1.51, group=Radial gradient, parameterName=Radial Gradient Density}
  LinearInterpolate(
   A=VectorParameter() {group=Colors, parameterName=RGB2, defaultValue=<rgba 0 0.0765 0.4271 1>}.RGB
   B=VectorParameter() {group=Colors, parameterName=RGB1, defaultValue=<rgba 0 0.0039 0.0706 1>}.RGB
   Alpha=Saturate(None=MaterialFunctionCall(UVs=@n1, CenterPosition=@n2, Radius=@n3, Density=@n4) {materialFunction=&RadialGradientExponential}.RadialGradientExponential)
  )
  n5 = ScalarParameter() {defaultValue=20, group=Layout, parameterName=TopLeftCornerRadius, sortPriority=4}
  n6 = ScalarParameter() {defaultValue=20, group=Layout, parameterName=BottomLeftCornerRadius, sortPriority=7}
  n7 = ScalarParameter() {defaultValue=20, group=Layout, parameterName=TopRightCornerRadius, sortPriority=5}
  n8 = ScalarParameter() {defaultValue=20, group=Layout, parameterName=BottomRightCornerRadius, sortPriority=6}
  n9 = ScalarParameter() {
   defaultValue=1
   desc="If enabled reads corner radius values in pixels.\r\nbool"
   group=Layout
   parameterName=UseValueInPixels
   sliderMax=1
   sortPriority=1
  }
  n10 = ComponentMask(None=MaterialFunctionCall!(UV=NamedRerouteUsage(), Size, Curvature=ScalarParameter() {defaultValue=0.25, group=Rounded corners, parameterName=Curvature}) {materialFunction=/Game/Effects/Fort_Effects/MaterialFunctions/MF_RoundedSquare}.Linear Gradient) {g=true}
  n11 = ScalarParameter() {defaultValue=32, group=Rounded corners, parameterName=EdgeSharpness}
  Multiply(
   A=MaterialFunctionCall(
    Corner Radius=MaterialFunctionCall(R=@n5, G=@n6, B=@n7, A=@n8) {materialFunction=&AppendMany}.RGBA
    Radius in Pixels=@n9
   ) {materialFunction=&MF_UI_CoreUI_RoundedMask}.Shape Mask
   B=Saturate(None=Multiply(A=@n10, B=@n11))
  )
  n12 = MaterialFunctionCall() {materialFunction=&GetUserInterfaceUV}
  NamedRerouteDeclaration(None=@n12.Normalized UV) {name=normalized uv, nodeColor=<rgba 1 0 0.3265 1>}
  NamedRerouteDeclaration(None=@n12."9-Slice UV (No Tiling)") {name=9 slice uv, nodeColor=<rgba 0.4531 0 1 1>}
  n13 = MaterialFunctionCall(UV=MaterialFunctionCall() {materialFunction=&GetUserInterfaceUV}."9-Slice UV (No Tiling)") {materialFunction=&MF_RoundedSquare}
  Saturate(None=Multiply(
    A=ComponentMask(None=@n13.Linear Gradient) {g=true}
    B=ScalarParameter() {defaultValue=32, group=Rounded corners, parameterName=EdgeSharpness}
   ))
 ]
}
