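# Translucent UI material for the quest container background in the minimap HUD, with hover and selected states.
# Color layers a base/hover blend, a two-tone selected fill patterned by a screen-aligned, aspect-corrected panning triangle grid, then right and left slices.
# Opacity combines state and slice values, cut by a rounded-square mask and an angled bottom cut.
# A static switch chooses whether the left slice stays visible or animates in with Select.
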Material /Game/Athena/HUD/MiniMap/Art/M_UI_Quest_Container_S19
graph {
 properties {blendMode=BLEND_Translucent, materialDomain=MD_UI, pixelDepthOffsetMode=PDOM_Legacy, usedShadingModels=Default Lit}
 nodes [
  n1 = LinearInterpolate(
   A=VectorParameter() {parameterName=Base_color, defaultValue=<rgba 0 0.004 0.0723 1>}.RGB
   B=VectorParameter() {parameterName=Hover_color, defaultValue=<rgba 0 0.06 0.6 1>}.RGB
   Alpha=ScalarParameter() {parameterName=Hover, sliderMax=1}
  ) {desc=is hovered?}
  n2 = MaterialFunctionCall(Float2=MaterialFunctionCall() {materialFunction=&ScreenResolution}.Visible Resolution) {materialFunction=&BreakOutFloat2Components}
  n3 = AppendVector(A=Divide(A=@n2.R, B=Max(A=@n2.G)), B=Constant() {r=1})
  n4 = MaterialFunctionCall() {materialFunction=&ScreenAlignedUVs}
  n5 = Panner(
   Coordinate=Rotator(Coordinate=Multiply(A=@n3, B=@n4."X 100%, Y 100%"), Time=ScalarParameter() {defaultValue=-0.0571, group=Triangles, parameterName=Tri Angle}) {speed=6.2832}
   Speed=AppendVector(A=ScalarParameter() {defaultValue=0.0025, group=Triangles, parameterName=PanSpeed}, B=Constant())
  )
  n6 = ScalarParameter() {defaultValue=5, group=Triangles, parameterName=Tri Tile Amount}
  n7 = Multiply(
   A=TextureSample(UVs=Multiply(A=@n5, B=@n6)) {texture=/Game/Global/Textures/Default/Masks_and_Gradients/T-TriangleGrid}.R
   B=ScalarParameter() {defaultValue=0.2, group=Triangles, parameterName=TriangleDensity}
  )
  n8 = NamedRerouteUsage()
  n9 = ScalarParameter() {group=Triangles, parameterName=TriangleMask_Position}
  n10 = Multiply(
   A=ComponentMask(None=Subtract(A=@n8, B=@n9)) {r=true}
   B=ScalarParameter() {defaultValue=1.5, group=Triangles, parameterName=TriangleMask_sharpness}
  )
  n11 = LinearInterpolate(
   A=VectorParameter() {group=colors, parameterName=Selected_RGB1, defaultValue=<rgba 1 1 1 1>}.RGB
   B=VectorParameter() {group=colors, parameterName=Selected_RGB2, defaultValue=<rgba 0.313 0.3397 0.4323 1>}.RGB
   Alpha=Multiply(A=Saturate(None=@n7), B=Saturate(None=@n10))
  )
  n12 = ScalarParameter() {parameterName=Select, sliderMax=1}
  n13 = Subtract(A=NamedRerouteUsage(), B=ScalarParameter() {defaultValue=0.75, group=RightSlice, parameterName=RightSlice_Left})
  n14 = Saturate(None=Multiply(A=ComponentMask(None=@n13) {r=true}, B=ScalarParameter() {defaultValue=32, group=RightSlice, parameterName=RightSlice_Sharpness}))
  n15 = ScalarParameter() {defaultValue=0.1, group=Left slice, parameterName=LeftSlice_Position}
  n16 = ScalarParameter() {parameterName=Select, sliderMax=1}
  n17 = Subtract(
   A=NamedRerouteUsage()
   B=StaticSwitchParameter(True=@n15, False=LinearInterpolate(B=@n15, Alpha=@n16) {constA=-0.1}) {group=Left slice, parameterName=Should Show Slice When Unselected}
  )
  n18 = ScalarParameter() {group=Left slice, parameterName=LeftSlice_Angle}
  n19 = StaticSwitchParameter(True=@n18, False=LinearInterpolate(B=@n18, Alpha=ScalarParameter() {parameterName=Select, sliderMax=1})) {group=Left slice, parameterName=Should Show Slice When Unselected}
  n20 = Multiply(
   A=ComponentMask(None=Rotator(Coordinate=@n17, Time=@n19)) {r=true}
   B=ScalarParameter() {defaultValue=128, group=Left slice, parameterName=LeftSlice_Sharpness}
  )
  n21 = Multiply(A=Saturate(None=OneMinus(None=@n20)), B=ScalarParameter() {defaultValue=1, group=Left slice, parameterName=LeftSlice_Opacity, sliderMax=1}) {constB=0.1}
  LinearInterpolate(
   A=LinearInterpolate(
    A=LinearInterpolate(A=@n1, B=@n11, Alpha=@n12) {desc=is selected?}
    B=VectorParameter() {group=RightSlice, parameterName=RightSlice_color, defaultValue=<rgba 0 0.004 0.0723 1>}.RGB
    Alpha=@n14
   ) {desc=right slice}
   B=VectorParameter() {group=Left slice, parameterName=RGB _LeftSlice, defaultValue=<rgba 1 0.9334 0 0>}.RGB
   Alpha=@n21
  ) {desc=left slice}
  n22 = LinearInterpolate(A=ScalarParameter() {defaultValue=0.5, parameterName=Opacity, sliderMax=1}, Alpha=ScalarParameter() {parameterName=Select, sliderMax=1})
  n23 = ScalarParameter() {defaultValue=0.85, group=RightSlice, parameterName=RightSlice_Opacity, sliderMax=1}
  n24 = Saturate(None=Add(A=@n21, B=LinearInterpolate(A=@n22, B=@n23, Alpha=@n14)))
  n25 = MaterialFunctionCall!(UV=NamedRerouteUsage(), Size, Curvature=ScalarParameter() {defaultValue=0.25, group=Rounded corners, parameterName=Curvature}) {materialFunction=/Game/Effects/Fort_Effects/MaterialFunctions/MF_RoundedSquare}
  n26 = Saturate(None=Multiply(
    A=ComponentMask(None=@n25.Linear Gradient) {g=true}
    B=ScalarParameter() {defaultValue=32, group=Rounded corners, parameterName=EdgeSharpness}
   ))
  n27 = Subtract(A=NamedRerouteUsage(), B=ScalarParameter() {defaultValue=1.1, group=BottomCut, parameterName=BottomCut_position})
  n28 = ScalarParameter() {group=BottomCut, parameterName=BottomCut_Angle}
  n29 = Multiply(
   A=ComponentMask(None=Rotator(Coordinate=@n27, Time=@n28)) {g=true}
   B=ScalarParameter() {defaultValue=32, group=BottomCut, parameterName=BottomCut_Sharpness}
  )
  Multiply(A=Saturate(None=Multiply(A=@n24, B=@n26)), B=OneMinus(None=Saturate(None=@n29)))
  n30 = MaterialFunctionCall() {materialFunction=&GetUserInterfaceUV}
  NamedRerouteDeclaration(None=@n30.Normalized UV) {name=normalized uv, nodeColor=<rgba 1 0 0.3265 1>}
  NamedRerouteDeclaration(None=@n30."9-Slice UV (No Tiling)") {name=9 slice uv, nodeColor=<rgba 0.4531 0 1 1>}
  n31 = MaterialFunctionCall(UV=MaterialFunctionCall() {materialFunction=&GetUserInterfaceUV}."9-Slice UV (No Tiling)") {materialFunction=&MF_RoundedSquare}
  Saturate(None=Multiply(
    A=ComponentMask(None=@n31.Linear Gradient) {g=true}
    B=ScalarParameter() {defaultValue=32, group=Rounded corners, parameterName=EdgeSharpness}
   ))
 ]
}
